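# Physics asset for the NobleCrusher backpack cosmetic, providing capsule bodies and constraints for its dynamic cape.
# The cape has center, left and right bone chains, each hanging from a tiny master capsule, with tight asymmetric swing limits and zero twist.
# Angularly free constraints tie the second left and right segments to the second center segment, linking the chains sideways.
# Larger pelvis and spine capsules, joined by another free constraint, represent the body with no constraint to the cape.
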PhysicsAsset /BRCosmetics/Accessories/FORT_Backpacks/Backpack_M_MED_NobleCrusher/Meshes/M_MED_NobleCrusher_Pack_Physics
bodies {
 Bsckpack_pelvice [{shapeType=Capsule, length=8, radius=15, shapeName=Bsckpack_pelvice_capsule, center=<-5 0 0>}]
 Bsckpack_spine_01 [{shapeType=Capsule, length=9, radius=12.5, shapeName=Bsckpack_spine_01_capsule}]
 dyn_cape_01 [{shapeType=Capsule, length=4, radius=5, shapeName=dyn_cape_01_capsule, center=<6.7492 0 2.6739>, rotation=<rot 90 0 0>}]
 dyn_cape_02 [{shapeType=Capsule, length=4, radius=5, shapeName=dyn_cape_02_capsule, center=<6.7492 0 2.6739>, rotation=<rot 90 0 0>}]
 dyn_cape_03 [{shapeType=Capsule, length=4, radius=5, shapeName=dyn_cape_03_capsule, center=<6.7492 0 2.6739>, rotation=<rot 90 0 0>}]
 dyn_cape_MSTR [{shapeType=Capsule, length=0.505, radius=0.505, shapeName=dyn_cape_MSTR_capsule, rotation=<rot 0 -90 0>}]
 dyn_cape_l_01 [{shapeType=Capsule, length=4, radius=5, shapeName=dyn_cape_l_01_capsule, center=<6.7492 0 2.6739>, rotation=<rot 90 0 0>}]
 dyn_cape_l_02 [{shapeType=Capsule, length=4, radius=5, shapeName=dyn_cape_l_02_capsule, center=<6.7492 0 2.6739>, rotation=<rot 90 0 0>}]
 dyn_cape_l_MSTR1 [{shapeType=Capsule, length=0.505, radius=0.505, shapeName=dyn_cape_l_MSTR1_capsule}]
 dyn_cape_r_01 [{shapeType=Capsule, length=4, radius=5, shapeName=dyn_cape_r_01_capsule, center=<6.7492 0 2.6739>, rotation=<rot 90 0 0>}]
 dyn_cape_r_02 [{shapeType=Capsule, length=4, radius=5, shapeName=dyn_cape_r_02_capsule, center=<6.7492 0 2.6739>, rotation=<rot 90 0 0>}]
 dyn_cape_r_MSTR [{shapeType=Capsule, length=0.505, radius=0.505, shapeName=dyn_cape_r_MSTR_capsule}]
}
constraints [
 {bone1Name=dyn_cape_03, bone2Name=dyn_cape_02, swing1LimitDegrees=12, swing2LimitDegrees=30, twistLimitDegrees=0}
 {bone1Name=dyn_cape_02, bone2Name=dyn_cape_01, swing1LimitDegrees=12, swing2LimitDegrees=30, twistLimitDegrees=0}
 {bone1Name=dyn_cape_01, bone2Name=dyn_cape_MSTR, swing1LimitDegrees=12, swing2LimitDegrees=30, twistLimitDegrees=0}
 {bone1Name=dyn_cape_l_01, bone2Name=dyn_cape_l_MSTR1, swing1LimitDegrees=12, swing2LimitDegrees=30, twistLimitDegrees=0}
 {bone1Name=dyn_cape_l_02, bone2Name=dyn_cape_l_01, swing1LimitDegrees=12, swing2LimitDegrees=30, twistLimitDegrees=0}
 {bone1Name=dyn_cape_r_01, bone2Name=dyn_cape_r_MSTR, swing1LimitDegrees=12, swing2LimitDegrees=30, twistLimitDegrees=0}
 {bone1Name=dyn_cape_r_02, bone2Name=dyn_cape_r_01, swing1LimitDegrees=12, swing2LimitDegrees=30, twistLimitDegrees=0}
 {
  bone1Name=dyn_cape_l_02
  bone2Name=dyn_cape_02
  swing1LimitDegrees=45
  swing1Motion=Free
  swing2LimitDegrees=45
  swing2Motion=Free
  twistLimitDegrees=45
  twistMotion=Free
 }
 {
  bone1Name=dyn_cape_r_02
  bone2Name=dyn_cape_02
  swing1LimitDegrees=45
  swing1Motion=Free
  swing2LimitDegrees=45
  swing2Motion=Free
  twistLimitDegrees=45
  twistMotion=Free
 }
 {
  bone1Name=Bsckpack_spine_01
  bone2Name=Bsckpack_pelvice
  swing1LimitDegrees=45
  swing1Motion=Free
  swing2LimitDegrees=45
  swing2Motion=Free
  twistLimitDegrees=45
  twistMotion=Free
 }
]
