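# Physics asset for the cape of the NobleCrusher backpack cosmetic, named for the M_MED body variant.
# It defines three capsule bodies: two thick ones on the backpack's pelvis and spine_01 bones, and a slim, slightly tilted, offset one on cape_top.
# A single constraint links spine_01 to pelvis with free swing and twist, so its angular limits are inactive.
# The cape_top body has no constraint.
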PhysicsAsset /BRCosmetics/Accessories/FORT_Backpacks/Backpack_M_MED_NobleCrusher/Meshes/M_MED_NobleCrusher_Pack_Physics_Cape
bodies {
 Bsckpack_pelvice [{shapeType=Capsule, length=5, radius=16.5, shapeName=Bsckpack_pelvice_capsule, center=<-6 0 0>}]
 Bsckpack_spine_01 [{shapeType=Capsule, length=8.7, radius=11.3, shapeName=Bsckpack_spine_01_capsule}]
 cape_top [{shapeType=Capsule, length=20, radius=3, shapeName=cape_top_capsule, center=<0 1.6273 -13.9143>, rotation=<rot 0 0 -5>}]
}
constraints [{
  bone1Name=Bsckpack_spine_01
  bone2Name=Bsckpack_pelvice
  swing1LimitDegrees=45
  swing1Motion=Free
  swing2LimitDegrees=45
  swing2Motion=Free
  twistLimitDegrees=45
  twistMotion=Free
 }]
