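# Animation Blueprint for the NobleCrusher backpack, driving its cape by cloth, rigid body or AnimDynamics simulation.
# A cached base pose offsets the root for medium female, front end, skydiving and spine twist, then snaps helper bones to the parent mesh's pelvis and spine_01.
# Cloth and rigid-body flags choose between that pose, RigidBody and a sphere-limited AnimDynamics chain.
# Each update runs either the gravity override or cloth controller, based on LOD, cloth and front-end state.
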AnimBlueprint /BRCosmetics/Accessories/FORT_Backpacks/Backpack_M_MED_NobleCrusher/Meshes/M_MED_NobleCrusher_Pack_AnimBP
target_skeleton=&/M_MED_NobleCrusher_Pack_Skeleton
graphs [
 AnimGraph {nodes [
   RigidBody() {
    overridePhysicsAsset=&/M_MED_NobleCrusher_Pack_Physics_Cape
    simSpaceSettings {dampingAlpha=0, worldAlpha=1}
   }
   n1 = RigidBody!(
    ComponentPose=Local To Component(LocalPose=Use cached pose 'Base'())
    Alpha
    OverrideWorldGravity:Vector="0,0,0"
    ExternalForce
    SimSpaceSettings:Sim Space Settings Structure=(WorldAlpha=1,VelocityScaleZ=1,DampingAlpha=0,MaxLinearVelocity=10000,MaxAngularVelocity=10000,MaxLinearAcceleration=10000,MaxAngularAcceleration=10000,ExternalLinearDragV=(X=0,Y=0,Z=0),ExternalLinearVelocity=(X=0,Y=0,Z=0),ExternalAngularVelocity=(X=0,Y=0,Z=0))
   ) -> (Pose) {
    bOverrideWorldGravity=true
    overridePhysicsAsset=&/M_MED_NobleCrusher_Pack_Physics
    simulationSpace=BaseBoneSpace
    baseBoneRef {boneName=root}
    simSpaceSettings {dampingAlpha=0, worldAlpha=1}
   }
   n2 = Component To Local(ComponentPose=@n1)
   n3 = Use cached pose 'Base'()
   n4 = Get bClothEnabled()
   n5 = Anim Dynamics - Bone: dyn_cape_r_01(ComponentPose=Local To Component(LocalPose=Use cached pose 'Base'()), Alpha=1, ExternalForce="0,0,0") {
    angularDampingOverride=0.999
    angularSpringConstant=150
    bAngularSpring=true
    bGravityOverrideInSimSpace=true
    bOverrideAngularDamping=true
    bUseGravityOverride=true
    simulationSpace=RootRelative
    boundBone {boneName=dyn_cape_r_01}
    componentAppliedLinearAccClamp=<2000 2000 2000>
    componentLinearAccScale=<0.5 0.5 0.5>
    componentLinearVelScale=<0.5 0.5 0.5>
    gravityOverride {z=-980}
    physicsBodyDefinitions [{
      boundBone {boneName=dyn_cape_r_01}
      boxExtents=<10 10 10>
      constraintSetup {angularLimitsMax=<0 0 15>, angularLimitsMin=<0 -45 -15>, angularTarget=<1 0 0>}
      localJointOffset=<1 0 0>
     }]
   }
   n6 = Anim Dynamics - Bone: dyn_cape_r_02(ComponentPose=@n5, Alpha=1, ExternalForce="0,0,0") {
    angularDampingOverride=0.999
    angularSpringConstant=150
    bAngularSpring=true
    bGravityOverrideInSimSpace=true
    bOverrideAngularDamping=true
    bUseGravityOverride=true
    bUseSphericalLimits=true
    simulationSpace=RootRelative
    boundBone {boneName=dyn_cape_r_02}
    componentAppliedLinearAccClamp=<2000 2000 2000>
    componentLinearAccScale=<0.5 0.5 0.5>
    componentLinearVelScale=<0.5 0.5 0.5>
    gravityOverride {z=-980}
    physicsBodyDefinitions [{
      collisionType=CustomSphere
      sphereCollisionRadius=5
      boundBone {boneName=dyn_cape_r_02}
      boxExtents=<10 10 10>
      constraintSetup {angularLimitsMax=<0 0 15>, angularLimitsMin=<0 -45 -15>, angularTarget=<1 0 0>}
      localJointOffset=<1 0 0>
     }]
    sphericalLimits [{limitRadius=12, drivingBone {boneName=Bsckpack_pelvice}, sphereLocalOffset=<0 0 0>}]
   }
   n7 = Anim Dynamics - Bone: dyn_cape_01(ComponentPose=@n6, Alpha=1, ExternalForce="0,0,0") {
    angularDampingOverride=0.999
    angularSpringConstant=150
    bAngularSpring=true
    bGravityOverrideInSimSpace=true
    bOverrideAngularDamping=true
    bUseGravityOverride=true
    simulationSpace=RootRelative
    boundBone {boneName=dyn_cape_01}
    componentAppliedLinearAccClamp=<2000 2000 2000>
    componentLinearAccScale=<0.5 0.5 0.5>
    componentLinearVelScale=<0.5 0.5 0.5>
    gravityOverride {z=-980}
    physicsBodyDefinitions [{
      boundBone {boneName=dyn_cape_01}
      boxExtents=<10 10 10>
      constraintSetup {angularLimitsMax=<0 0 15>, angularLimitsMin=<0 -45 -15>, angularTarget=<1 0 0>}
      localJointOffset=<1 0 0>
     }]
   }
   n8 = Anim Dynamics - Bone: dyn_cape_02(ComponentPose=@n7, Alpha=1, ExternalForce="0,0,0") {
    angularDampingOverride=0.999
    angularSpringConstant=150
    bAngularSpring=true
    bGravityOverrideInSimSpace=true
    bOverrideAngularDamping=true
    bUseGravityOverride=true
    bUseSphericalLimits=true
    simulationSpace=RootRelative
    boundBone {boneName=dyn_cape_02}
    componentAppliedLinearAccClamp=<2000 2000 2000>
    componentLinearAccScale=<0.5 0.5 0.5>
    componentLinearVelScale=<0.5 0.5 0.5>
    gravityOverride {z=-980}
    physicsBodyDefinitions [{
      collisionType=CustomSphere
      sphereCollisionRadius=5
      boundBone {boneName=dyn_cape_02}
      boxExtents=<10 10 10>
      constraintSetup {angularLimitsMax=<0 0 15>, angularLimitsMin=<0 -45 -15>, angularTarget=<1 0 0>}
      localJointOffset=<1 0 0>
     }]
    sphericalLimits [{limitRadius=12, drivingBone {boneName=Bsckpack_pelvice}, sphereLocalOffset=<0 0 0>}]
   }
   n9 = Anim Dynamics - Bone: dyn_cape_03(ComponentPose=@n8, Alpha=1, ExternalForce="0,0,0") {
    angularDampingOverride=0.999
    angularSpringConstant=150
    bAngularSpring=true
    bGravityOverrideInSimSpace=true
    bOverrideAngularDamping=true
    bUseGravityOverride=true
    bUseSphericalLimits=true
    simulationSpace=RootRelative
    boundBone {boneName=dyn_cape_03}
    componentAppliedLinearAccClamp=<2000 2000 2000>
    componentLinearAccScale=<0.5 0.5 0.5>
    componentLinearVelScale=<0.5 0.5 0.5>
    gravityOverride {z=-980}
    physicsBodyDefinitions [{
      collisionType=CustomSphere
      sphereCollisionRadius=5
      boundBone {boneName=dyn_cape_03}
      boxExtents=<10 10 10>
      constraintSetup {angularLimitsMax=<0 0 15>, angularLimitsMin=<0 -45 -15>, angularTarget=<1 0 0>}
      localJointOffset=<1 0 0>
     }]
    sphericalLimits [{limitRadius=1.3, drivingBone {boneName=Bsckpack_pelvice}, sphereLocalOffset=<0 0 0>}]
   }
   n10 = Anim Dynamics - Bone: dyn_cape_l_01(ComponentPose=@n9, Alpha=1, ExternalForce="0,0,0") {
    angularDampingOverride=0.999
    angularSpringConstant=150
    bAngularSpring=true
    bGravityOverrideInSimSpace=true
    bOverrideAngularDamping=true
    bUseGravityOverride=true
    simulationSpace=RootRelative
    boundBone {boneName=dyn_cape_l_01}
    componentAppliedLinearAccClamp=<2000 2000 2000>
    componentLinearAccScale=<0.5 0.5 0.5>
    componentLinearVelScale=<0.5 0.5 0.5>
    gravityOverride {z=-980}
    physicsBodyDefinitions [{
      boundBone {boneName=dyn_cape_l_01}
      boxExtents=<10 10 10>
      constraintSetup {angularLimitsMax=<0 0 15>, angularLimitsMin=<0 -45 -15>, angularTarget=<1 0 0>}
      localJointOffset=<1 0 0>
     }]
   }
   n11 = Anim Dynamics - Bone: dyn_cape_l_02(ComponentPose=@n10, Alpha=1, ExternalForce="0,0,0") {
    angularDampingOverride=0.999
    angularSpringConstant=150
    bAngularSpring=true
    bGravityOverrideInSimSpace=true
    bOverrideAngularDamping=true
    bUseGravityOverride=true
    bUseSphericalLimits=true
    simulationSpace=RootRelative
    boundBone {boneName=dyn_cape_l_02}
    componentAppliedLinearAccClamp=<2000 2000 2000>
    componentLinearAccScale=<0.5 0.5 0.5>
    componentLinearVelScale=<0.5 0.5 0.5>
    gravityOverride {z=-980}
    physicsBodyDefinitions [{
      collisionType=CustomSphere
      sphereCollisionRadius=5
      boundBone {boneName=dyn_cape_l_02}
      boxExtents=<10 10 10>
      constraintSetup {angularLimitsMax=<0 0 15>, angularLimitsMin=<0 -45 -15>, angularTarget=<1 0 0>}
      localJointOffset=<1 0 0>
     }]
    sphericalLimits [{limitRadius=12, drivingBone {boneName=Bsckpack_pelvice}, sphereLocalOffset=<0 0 0>}]
   }
   Output Pose(Result=Blend Poses by bool(
     BlendPose_0=Blend Poses by bool(BlendPose_0=@n3, BlendPose_1=@n2, bActiveValue=@n4)
     BlendPose_1=Component To Local(ComponentPose=@n11)
     bActiveValue=Get bRigidBodyEnabled()
    )) {layerGroup=None}
   n12 = "Transform (Modify) Bone - Bone: root"!(
    ComponentPose=Local To Component(LocalPose=Input Pose())
    bAlphaBoolEnabled:bool=Get Not_FE_and_F() -> (Not_FE_and_F:bool)
    Translation="0,0,0"
    Rotation="0,0,7"
    Scale
   ) -> (Pose) {
    alphaInputType=Bool
    rotationMode=BMM_Additive
    scaleMode=BMM_Additive
    translationMode=BMM_Additive
    boneToModify {boneName=root}
    rotation {roll=7}
   }
   n13 = "Transform (Modify) Bone - Bone: root"(ComponentPose=@n12, bAlphaBoolEnabled=Get bIsInFrontEnd(), Translation="0,0,0", Rotation="0,0,10") {
    alphaInputType=Bool
    rotationMode=BMM_Additive
    scaleMode=BMM_Additive
    translationMode=BMM_Additive
    boneToModify {boneName=root}
    rotation {roll=10}
   }
   n14 = "Transform (Modify) Bone - Bone: root"(ComponentPose=@n13, bAlphaBoolEnabled=Get bIsSkyDiving(), Translation="0,-2,1", Rotation="0,0,20") {
    alphaInputType=Bool
    rotationMode=BMM_Additive
    translationMode=BMM_Additive
    alphaBoolBlend {blendOption=Linear}
    boneToModify {boneName=root}
    rotation {roll=20}
    translation {y=-2, z=1}
   }
   n15 = "Transform (Modify) Bone - Bone: root"!(ComponentPose=@n14, Alpha:float, Translation="0,-17,10", Rotation="0,0,0", Scale) -> (Pose) {translationMode=BMM_Additive, alphaBoolBlend {blendOption=Linear}, boneToModify {boneName=root}, translation {y=-17, z=10}}
   n16 = "Transform (Modify) Bone - Bone: root"(ComponentPose=@n15, Alpha=Get Picckaxe() -> (Picckaxe:double), Translation="0,0,0", Rotation="0,0,10") {
    rotationMode=BMM_Additive
    scaleMode=BMM_Additive
    translationMode=BMM_Additive
    alphaBoolBlend {blendOption=Linear}
    boneToModify {boneName=root}
    rotation {roll=10}
   }
   n17 = "Transform (Modify) Bone - Bone: Bsckpack_pelvice"!(
    ComponentPose=@n16
    bAlphaBoolEnabled:bool=True
    Translation=Get PelvisLoc() -> (PelvisLoc:Vector)
    Rotation=Get PelvisRot() -> (PelvisRot:Rotator)
    Scale=Get PelvisScale() -> (PelvisScale:Vector)
   ) -> (Pose) {
    alphaInputType=Bool
    rotationMode=BMM_Replace
    rotationSpace=BCS_WorldSpace
    scaleMode=BMM_Replace
    scaleSpace=BCS_WorldSpace
    translationMode=BMM_Replace
    translationSpace=BCS_WorldSpace
    alphaBoolBlend {blendInTime=0.2, blendOption=Linear, blendOutTime=0.2}
    boneToModify {boneName=Bsckpack_pelvice}
    rotation {roll=25}
   }
   n18 = "Transform (Modify) Bone - Bone: Bsckpack_spine_01"!(
    ComponentPose=@n17
    bAlphaBoolEnabled:bool=True
    Translation=Get spine_01_Loc() -> (spine_01_Loc:Vector)
    Rotation=Get spine_01_Rot() -> (spine_01_Rot:Rotator)
    Scale=Get spine_01_Scale() -> (spine_01_Scale:Vector)
   ) -> (Pose) {
    alphaInputType=Bool
    rotationMode=BMM_Replace
    rotationSpace=BCS_WorldSpace
    scaleMode=BMM_Replace
    scaleSpace=BCS_WorldSpace
    translationMode=BMM_Replace
    translationSpace=BCS_WorldSpace
    alphaBoolBlend {blendInTime=0.2, blendOption=Linear, blendOutTime=0.2}
    boneToModify {boneName=Bsckpack_spine_01}
   }
   Save cached pose 'Base'(Pose=Component To Local(ComponentPose=@n18)) {cacheName=Base}
  ]}
 EventGraph {nodes [
   n1 = Event BlueprintUpdateAnimation()
   > Set Delta Time X(Delta Time X=@n1.DeltaTimeX)
   > Set Lod 0(Lod 0:bool="integer > integer"(A=Get CurrentLOD())) -> (Output_Get:bool)
   n2 = Get bIsInFrontEnd()
   > Set Not_FE_and_F(Not_FE_and_F:bool=AND Boolean(A=NOT Boolean(A=@n2), B=Get isF_Med() -> (isF_Med:bool))) -> (Output_Get:bool)
   > n3 = Sequence!(execute) -> (then_0, then_1, then_2:Exec)
   n4 = Get bClothEnabled()
   > n5 = Branch(Condition=OR Boolean!(A=Get Lod 0() -> (Lod 0:bool), B=NOT Boolean(A=@n4), C:bool=Get bIsInFrontEnd()) -> (ReturnValue))
   > n6 = "Gravity Override State Function (Parameter Asset) (Nativized)"(
    CharacterPartAnimInstance=Self-Reference()
    GameTimeInSeconds=GetGameTimeInSeconds()
    Parameters=&/M_MED_NobleCrusher_Pack_Parameters_01
   )
   > Set Noncloth Gravity Override(Noncloth Gravity Override:Vector=@n6) -> (Output_Get:Vector)
   > Set RBAN Settings Out(RBAN Settings Out=@n6.OutRigidBodyAnimNodeParameters)
   "Cloth Controller Function (Parameter Asset) (Nativized)"!(
    execute=@n5.else
    CharacterPartAnimInstance:Custom Character Part Anim Instance Object Reference=Self-Reference()
    ClothingAssetName:Name=M_MED_NobleCrusher_Pack_Clothing
    ClothCoreSettings:"Cloth Core Settings Structure (by ref)"=Get Cloth Core Settings() -> (Cloth Core Settings:Cloth Core Settings Structure)
    WindGustLocalSettings:"Wind Gust Settings Structure (by ref)"=Get Wind Gust Local Settings() -> (Wind Gust Local Settings:Wind Gust Settings Structure)
    ClothUpAxis:Vector="1,0,0"
    LinVelScaleFwdBwdClamp:float=0
    LinVelScaleUpDwnClamp:float=0.7
    DeltaTime:float=Get Delta Time X()
    GameTimeInSeconds:float=GetGameTimeInSeconds()
    StateLogicLayerName:Name=none
    Parameters:Fort Character Dynamics Parameters Object Reference=&M_MED_NobleCrusher_Pack_Parameters_01
   ) -> (
    then
    OutCurrentState:Name=None
    OutClothParameters:Cloth Parameters Structure
    OutClothCoreSettings:Cloth Core Settings Structure
    OutWindGustSettings:Wind Gust Parameters Structure
    OutWindGustLocalSettings:Wind Gust Settings Structure
    OutWindVelocity:Vector
    OutLinearVelVector:Vector
   )
   n7 = GetOwnerSkeletalMeshComponent()
   n8 = BreakRotator(InRot=BreakTransform(InTransform=GetDeltaTransformFromRefPose(self=@n7, BoneName=spine_05, BaseName=spine_01)).Rotation)
   Set BackpackRotate(execute=@n3.then_2, BackpackRotate:double=MapRangeClamped(Value=@n8.Yaw, InRangeA=-25, InRangeB=-65)) -> (Output_Get:double)
   > Set Picckaxe(Picckaxe:double=float + float(A=MapRangeClamped(Value=@n8, InRangeA=-25, InRangeB=-45), B=MapRangeClamped(Value=@n8, InRangeA=25, InRangeB=45))) -> (Output_Get:double)
   n9 = GetPartSkeletalMeshComponent()
   Is Valid(exec=@n3.then_1, InputObject=@n9)
   n10 = GetAttachParent!(self:Scene Component Object Reference=@n9) -> (ReturnValue:Scene Component Object Reference)
   > Is Valid(InputObject=@n10)
   n11 = GetSocketTransform!(self=@n10, InSocketName=pelvis, TransformSpace=RTS_World) -> (ReturnValue_Location:Vector, ReturnValue_Rotation:Rotator, ReturnValue_Scale:Vector)
   > Set PelvisLoc(PelvisLoc:Vector=@n11) -> (Output_Get:Vector)
   > Set PelvisRot(PelvisRot:Rotator=@n11.ReturnValue_Rotation) -> (Output_Get:Rotator)
   > Set PelvisScale(PelvisScale:Vector=@n11.ReturnValue_Scale) -> (Output_Get:Vector)
   n12 = GetSocketTransform(self=@n10, InSocketName=spine_01, TransformSpace=RTS_World)
   > Set spine_01_Loc(spine_01_Loc:Vector=@n12) -> (Output_Get:Vector)
   > Set spine_01_Rot(spine_01_Rot:Rotator=@n12.ReturnValue_Rotation) -> (Output_Get:Rotator)
   > Set spine_01_Scale(spine_01_Scale:Vector=@n12.ReturnValue_Scale) -> (Output_Get:Vector)
   n13 = Event OnCharacterCustomizationLoaded()
   n14 = "Get (a copy)"(Array:Array of Custom Character Part Object References=@n13.CustomParts, Dimension 1=1) -> (Output:Custom Character Part Object Reference)
   n15 = Get GenderPermitted!(self:Custom Character Part Object Reference=@n14) -> (GenderPermitted:EFortCustomGender Enum=Invalid)
   n16 = Get BodyTypesPermitted!(self:Custom Character Part Object Reference=@n14) -> (BodyTypesPermitted:EFortCustomBodyType Enum=NONE)
   > Set isF_Med(isF_Med:bool=AND Boolean(
     A="Equal (Enum)"(A:EFortCustomBodyType Enum=@n16, B:EFortCustomBodyType Enum=Medium)
     B="Equal (Enum)"(A:EFortCustomGender Enum=@n15, B:EFortCustomGender Enum=Female)
    )) -> (Output_Get:bool)
   TryGetPawnOwner()
   n17 = GetSocketTransform(InSocketName=pelvis, TransformSpace=RTS_World)
   Set spine_02_Scale(spine_02_Scale:Vector=@n17.ReturnValue_Scale) -> (Output_Get:Vector)
   Set spine_02_Loc(spine_02_Loc:Vector=@n17) -> (Output_Get:Vector)
   Set spine_02_Rot(spine_02_Rot:Rotator=@n17.ReturnValue_Rotation) -> (Output_Get:Rotator)
  ]}
]
variables [
 BackpackRotate {type=number}
 Cloth Core Settings {
  type=ClothCoreSettings
  default {
   collisionThickness=1.5
   dampingCoefficient=0.05
   frictionCoefficient=0.3
   localDampingCoefficient=0
   numberIterations=2
   numberSubsteps=1
   selfCollisionThickness=0
   animDriveDamping=<0 1>
   animDriveStiffness=<0 1>
   areaStiffness=<1.2 1.2>
   bendingStiffness=<1.2 1.2>
   edgeStiffness=<1.5 1.5>
   tetherStiffness=<1.2 1.2>
  }
 }
 Delta Time X {type=number}
 Is F Med {type=boolean}
 isF_Med {type=boolean}
 Lod 0 {type=boolean}
 NewVar {type=boolean}
 Noncloth Gravity Override {type=Vector}
 Not_FE_and_F {type=boolean}
 PelvisLoc {type=Vector}
 PelvisRot {type=Rotator}
 PelvisScale {type=Vector}
 Picckaxe {type=number}
 RBAN Settings Out {type=FortRigidBodyAnimNodeParameters}
 spine_01_Loc {type=Vector}
 spine_01_Rot {type=Rotator}
 spine_01_Scale {type=Vector}
 spine_02_Loc {type=Vector}
 spine_02_Rot {type=Rotator}
 spine_02_Scale {type=Vector}
 Wind Gust Local Settings {type=WindGustSettings, default {elapsedTime=0, finalGustStrengthLocal=0, gustDuration=0, gustStrength=0, timeBetweenGusts=0}}
]
